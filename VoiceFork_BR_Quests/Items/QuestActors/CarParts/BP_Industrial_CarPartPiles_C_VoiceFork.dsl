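; Industrial car part pile quest actor that gates its damageability on whether incoming damage comes from a pickaxe.
; It starts damageable at BeginPlay, and the server damage event compares the hit's gameplay tags with its pickaxe tag container.
; A match keeps it damageable; any other hit switches off bCanBeDamaged for a few seconds through a Delay, then restores it.
; Overlap, Tick and the construction script add no logic of their own.
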(event Collision|EventActorBeginOverlap
  (params (OtherActor :type Actor%20Object%20Reference))
  (|CallParentFunction OtherActor) ; This node is disabled and will not be called.\nDrag off pins to build functionality.
)

(event EventTick
  (params (DeltaSeconds :type Float%20%28single-precision%29))
  (|CallParentFunction DeltaSeconds) ; This node is disabled and will not be called.\nDrag off pins to build functionality.
)

(event EventOnDamageServer
  (params (Damage :type Float%20%28single-precision%29) (DamageTags :type Gameplay%20Tag%20Container%20Structure%20%28by%20ref%29) (Momentum :type Vector :default "0, 0, 0") (HitInfo :type Hit%20Result%20Structure%20%28by%20ref%29) (InstigatedBy :type Controller%20Object%20Reference) (DamageCauser :type Actor%20Object%20Reference) (EffectContext :type Gameplay%20Effect%20Context%20Handle%20Structure))
  (if (GameplayTags|TagContainer|CompareTagContainertoOtherTags (Variables|Default|GetPickaxeTag) DamageTags)
    (Variables|Actor|SetbCanBeDamaged true)
    (else
      (Variables|Actor|SetbCanBeDamaged false)
      (Utilities|FlowControl|Delay 3.0)
      (Variables|Actor|SetbCanBeDamaged true))))

(event EventBeginPlay
  (|CallParentFunction)
  (Variables|Actor|SetbCanBeDamaged true))

(fn ConstructionScript
  (|CallParentFunction))
